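# Translucent unlit particle material for Niagara and particle sprites that produces a hover refraction distortion effect.
# A downward-panned normal texture perturbs the UVs of a cone distortion texture, which supplies a distortion vector and an alpha mask.
# Index-of-refraction strength varies along V and blends in with particle alpha.
# Particle color feeds emissive adjustments, while Fresnel, texture alpha, a V falloff, OpacityMultiplier and particle alpha form a faint opacity.
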Material /BRCosmetics/Effects/Niagara/MasterMaterials/Refraction/M_HoverRefraction
graph {
 properties {
  bUsedWithNiagaraSprites=true
  bUsedWithParticleSprites=true
  blendMode=BLEND_Translucent
  pixelDepthOffsetMode=PDOM_Legacy
  refractionMethod=RM_IndexOfRefraction
  shadingModel=MSM_Unlit
  usedShadingModels=Unlit
 }
 nodes [
  n1 = TextureSample(UVs=Panner() {speedY=-2}) {samplerType=SAMPLERTYPE_Normal, texture=/Game/Packages/Fortress_Sky/Materials/2Normal}
  n2 = MaterialFunctionCall(X=@n1.R, Y=@n1.R) {materialFunction=&MakeFloat2}
  n3 = TextureSample(UVs=Add(A=TextureCoordinate(), B=Multiply(A=@n2.Result) {constB=0.1})) {samplerType=SAMPLERTYPE_LinearColor, texture=/Game/Missions/Primary/OB_FarmsteadFort/ArtAssets/Textures/T_DistortCone}
  n4 = MaterialFunctionCall(X=@n3.R, Y=@n3.R) {materialFunction=&MakeFloat2}
  AppendVector(A=Subtract(A=Multiply(A=@n4.Result) {constB=2}), B=@n3.B)
  n5 = ComponentMask(None=TextureCoordinate()) {g=true}
  n6 = Power(Base=@n5)
  n7 = ParticleColor()
  LinearInterpolate(
   B=LinearInterpolate(B=ScalarParameter() {defaultValue=0.98, parameterName=RefractStrength}, Alpha=OneMinus(None=@n6)) {constA=1}
   Alpha=@n7.A
  ) {constA=1}
  MaterialFunctionCall(Emissive=@n7.RGB) {materialFunction=&MF_FX_EmissiveAdjustments}
  n8 = Multiply(A=Multiply(A=MaterialFunctionCall() {materialFunction=&MF_Fresnel_FX}.Fresnel), B=@n3.A)
  n9 = OneMinus(None=Power(Base=@n5) {constExponent=4})
  Multiply(A=Multiply(A=Multiply(A=@n8, B=@n9), B=ScalarParameter() {defaultValue=0.03, parameterName=OpacityMultiplier}), B=@n7.A)
 ]
}
